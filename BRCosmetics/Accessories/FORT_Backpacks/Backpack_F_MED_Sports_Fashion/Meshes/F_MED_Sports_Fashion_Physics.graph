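# Physics asset for the F_MED_Sports_Fashion backpack cosmetic, giving its headphones, phone, cords and zipper limited physical swing.
# Three capsule cord bodies and box headphone and phone bodies each constrain directly to a small music base box, rather than chaining to one another.
# A separate zipper base box anchors the zipper body.
# Headphones and phone have very tight limits, cords and zipper swing more widely, and twist stays narrow throughout.
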PhysicsAsset /BRCosmetics/Accessories/FORT_Backpacks/Backpack_F_MED_Sports_Fashion/Meshes/F_MED_Sports_Fashion_Physics
bodies {
 dyn_cord_01 [{shapeType=Capsule, length=2.4065, radius=1.3481, center=<0 0 0.8848>, rotation=<rot 90 0 0>}]
 dyn_cord_02 [{shapeType=Capsule, length=2.6041, radius=1.3129, center=<0 0 0.4503>, rotation=<rot 90 0 0>}]
 dyn_cord_03 [{shapeType=Capsule, length=2.2608, radius=0.9498, center=<1.3729 0 0>, rotation=<rot 90 0 0>}]
 dyn_headphones_parent [{shapeType=Box, extentX=3.8881, extentY=5.945, extentZ=3.2657, center=<1.2228 0 0>}]
 dyn_music_base [{shapeType=Box, extentX=0.5003, extentY=0.5003, extentZ=0.5003}]
 dyn_phone [{shapeType=Box, extentX=3.5758, extentY=3.3395, extentZ=1.4191, center=<0.7334 0 0.0002>}]
 dyn_zipper_base [{shapeType=Box, extentX=0.5, extentY=0.5, extentZ=0.5}]
 dyn_zipper_parent [{shapeType=Box, extentX=4.0328, extentY=1.9777, extentZ=0.7229, center=<2.0388 0 0>}]
}
constraints [
 {bone1Name=dyn_cord_01, bone2Name=dyn_music_base, swing1LimitDegrees=30, swing2LimitDegrees=30, twistLimitDegrees=5}
 {bone1Name=dyn_cord_02, bone2Name=dyn_music_base, swing1LimitDegrees=25, swing2LimitDegrees=25, twistLimitDegrees=3}
 {bone1Name=dyn_cord_03, bone2Name=dyn_music_base, swing1LimitDegrees=25, swing2LimitDegrees=10, twistLimitDegrees=3}
 {bone1Name=dyn_headphones_parent, bone2Name=dyn_music_base, swing1LimitDegrees=4, swing2LimitDegrees=1, twistLimitDegrees=1}
 {bone1Name=dyn_phone, bone2Name=dyn_music_base, swing1LimitDegrees=4, swing2LimitDegrees=2, twistLimitDegrees=2}
 {bone1Name=dyn_zipper_parent, bone2Name=dyn_zipper_base, swing1LimitDegrees=30, swing2LimitDegrees=40, twistLimitDegrees=3}
]
