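# Physics asset for the straps of the F_MED Sports Fashion backpack cosmetic, defining collision bodies and joints for dangling left and right straps.
# Each strap is a chain of box bodies running from a small base through a parent to five segments, while a capsule covers the pack body.
# Angular constraints link neighbors with tight twist and moderate swing limits, locking one swing axis at both roots and the left tip.
# No constraint joins the pack capsule to either chain.
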PhysicsAsset /BRCosmetics/Accessories/FORT_Backpacks/Backpack_F_MED_Sports_Fashion/Meshes/F_MED_Sports_Fashion_Straps_Physics
bodies {
 dyn_strap_l_01 [{shapeType=Box, extentX=2.7216, extentY=4.2004, extentZ=1.0602, center=<1.3611 0 0>}]
 dyn_strap_l_02 [{shapeType=Box, extentX=2.9856, extentY=4.1972, extentZ=1.0602, center=<1.6588 0 0>}]
 dyn_strap_l_03 [{shapeType=Box, extentX=5.035, extentY=4.1972, extentZ=1.0602, center=<2.3256 0 0>}]
 dyn_strap_l_04 [{shapeType=Box, extentX=5.5489, extentY=4.2036, extentZ=1.0602, center=<2.8818 0 0>}]
 dyn_strap_l_05 [{shapeType=Box, extentX=2.9656, extentY=4.1972, extentZ=1.0602, center=<1.4219 0 0.0001>}]
 dyn_strap_l_base [{shapeType=Box, extentX=0.5001, extentY=0.5001, extentZ=0.5001}]
 dyn_strap_l_parent [{shapeType=Box, extentX=1.866, extentY=4.1972, extentZ=1.0602, center=<0.867 0 0>}]
 dyn_strap_r_01 [{shapeType=Box, extentX=5.2832, extentY=3.6396, extentZ=2.823, center=<1.5016 0 0>}]
 dyn_strap_r_02 [{shapeType=Box, extentX=2.6663, extentY=5.6395, extentZ=2.823, center=<1.4416 0 0>}]
 dyn_strap_r_03 [{shapeType=Box, extentX=4.2355, extentY=3.6386, extentZ=2.823, center=<1.4499 0 0>}]
 dyn_strap_r_04 [{shapeType=Box, extentX=2.6796, extentY=5.6316, extentZ=2.823, center=<1.4147 0 0>}]
 dyn_strap_r_05 [{shapeType=Box, extentX=5.4889, extentY=7.519, extentZ=2.823, center=<1.709 0 0>}]
 dyn_strap_r_base [{shapeType=Box, extentX=0.5001, extentY=0.5001, extentZ=0.5001}]
 dyn_strap_r_parent [{shapeType=Box, extentX=2.0056, extentY=3.6386, extentZ=2.823, center=<1.0506 0 0>}]
 pack [{shapeType=Capsule, length=14.2617, radius=9.2218, center=<-0.1679 7.6355 -15.9186>, rotation=<rot -89.2617 -179.9991 179.9991>}]
}
constraints [
 {
  bone1Name=dyn_strap_r_parent
  bone2Name=dyn_strap_r_base
  swing1LimitDegrees=20
  swing1Motion=Locked
  swing2LimitDegrees=40
  twistLimitDegrees=3
 }
 {bone1Name=dyn_strap_r_01, bone2Name=dyn_strap_r_parent, swing1LimitDegrees=20, swing2LimitDegrees=50, twistLimitDegrees=3}
 {bone1Name=dyn_strap_r_02, bone2Name=dyn_strap_r_01, swing1LimitDegrees=20, swing2LimitDegrees=50, twistLimitDegrees=3}
 {bone1Name=dyn_strap_r_03, bone2Name=dyn_strap_r_02, swing1LimitDegrees=30, swing2LimitDegrees=40, twistLimitDegrees=3}
 {bone1Name=dyn_strap_r_04, bone2Name=dyn_strap_r_03, swing1LimitDegrees=30, swing2LimitDegrees=40, twistLimitDegrees=3}
 {bone1Name=dyn_strap_r_05, bone2Name=dyn_strap_r_04, swing1LimitDegrees=30, swing2LimitDegrees=40, twistLimitDegrees=3}
 {
  bone1Name=dyn_strap_l_parent
  bone2Name=dyn_strap_l_base
  swing1LimitDegrees=20
  swing1Motion=Locked
  swing2LimitDegrees=40
  twistLimitDegrees=3
 }
 {bone1Name=dyn_strap_l_01, bone2Name=dyn_strap_l_parent, swing1LimitDegrees=20, swing2LimitDegrees=40, twistLimitDegrees=3}
 {bone1Name=dyn_strap_l_02, bone2Name=dyn_strap_l_01, swing1LimitDegrees=20, swing2LimitDegrees=30, twistLimitDegrees=3}
 {bone1Name=dyn_strap_l_03, bone2Name=dyn_strap_l_02, swing1LimitDegrees=20, swing2LimitDegrees=40, twistLimitDegrees=3}
 {bone1Name=dyn_strap_l_04, bone2Name=dyn_strap_l_03, swing1LimitDegrees=20, swing2LimitDegrees=40, twistLimitDegrees=3}
 {bone1Name=dyn_strap_l_05, bone2Name=dyn_strap_l_04, swing1LimitDegrees=20, swing1Motion=Locked, swing2LimitDegrees=40, twistLimitDegrees=3}
]
