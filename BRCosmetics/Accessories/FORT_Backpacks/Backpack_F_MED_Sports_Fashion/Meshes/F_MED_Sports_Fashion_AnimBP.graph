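# Animation Blueprint for the Sports Fashion backpack cosmetic, adding physics and movement-driven strap motion.
# The input pose gets a slight pack bone offset, then passes through two chained RigidBody simulations, one using the straps physics asset and one the main physics asset.
# Both strap base bones are then pitched additively, with interpolated alpha from PlayerMoving.
# The update event sets PlayerMoving to 1 or 0 by thresholding 2D pawn speed.
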AnimBlueprint /BRCosmetics/Accessories/FORT_Backpacks/Backpack_F_MED_Sports_Fashion/Meshes/F_MED_Sports_Fashion_AnimBP
target_skeleton=&/F_MED_Sports_Fashion_Skeleton
graphs [
 AnimGraph {nodes [
   n1 = Get PlayerMoving() -> (PlayerMoving:double)
   n2 = "Transform (Modify) Bone - Bone: pack"(ComponentPose=Local To Component(LocalPose=Input Pose()), Translation="0,0,-2.5", Rotation="0,0,0") {
    scaleMode=BMM_Replace
    scaleSpace=BCS_BoneSpace
    translationMode=BMM_Additive
    translationSpace=BCS_BoneSpace
    alphaBoolBlend {blendOption=Linear}
    boneToModify {boneName=pack}
    translation {z=-2.5}
   }
   n3 = RigidBody(ComponentPose=@n2) {
    bClampLinearTranslationLimitToRefPose=true
    bOverrideWorldGravity=true
    overridePhysicsAsset=&/F_MED_Sports_Fashion_Straps_Physics
    alphaBoolBlend {blendOption=Linear}
    baseBoneRef {boneName=root}
    componentAppliedLinearAccClamp=<3000 3000 3000>
    componentLinearAccScale=<0.25 0.25 0.28>
    componentLinearVelScale=<0.25 0.25 0.28>
    overrideWorldGravity {z=-980}
   }
   n4 = RigidBody(ComponentPose=@n3) {
    bClampLinearTranslationLimitToRefPose=true
    bOverrideWorldGravity=true
    overridePhysicsAsset=&/F_MED_Sports_Fashion_Physics
    simulationSpace=BaseBoneSpace
    alphaBoolBlend {blendOption=Linear}
    baseBoneRef {boneName=root}
    componentAppliedLinearAccClamp=<6000 6000 6000>
    componentLinearAccScale=<0.25 0.25 0.25>
    componentLinearVelScale=<0.25 0.25 0.25>
    overrideWorldGravity {z=-980}
   }
   n5 = "Transform (Modify) Bone - Bone: dyn_strap_r_base"(ComponentPose=@n4, Alpha=@n1, Translation="0,0,0", Rotation="-10,0,0", Scale="1,1,1") {
    rotationMode=BMM_Additive
    rotationSpace=BCS_BoneSpace
    scaleMode=BMM_Replace
    scaleSpace=BCS_BoneSpace
    translationMode=BMM_Additive
    translationSpace=BCS_BoneSpace
    alphaBoolBlend {blendOption=Linear}
    alphaScaleBiasClamp {bInterpResult=true, interpSpeedDecreasing=6, interpSpeedIncreasing=8}
    boneToModify {boneName=dyn_strap_r_base}
    rotation {pitch=-10}
   }
   n6 = "Transform (Modify) Bone - Bone: dyn_strap_l_base"(ComponentPose=@n5, Alpha=@n1, Translation="0,0,0", Rotation="-10,0,0", Scale="1,1,1") {
    rotationMode=BMM_Additive
    rotationSpace=BCS_BoneSpace
    scaleMode=BMM_Replace
    scaleSpace=BCS_BoneSpace
    translationMode=BMM_Additive
    translationSpace=BCS_BoneSpace
    alphaBoolBlend {blendOption=Linear}
    alphaScaleBiasClamp {bInterpResult=true, interpSpeedDecreasing=6, interpSpeedIncreasing=8}
    boneToModify {boneName=dyn_strap_l_base}
    rotation {pitch=-10}
   }
   Output Pose(Result=Component To Local(ComponentPose=@n6)) {layerGroup=None}
  ]}
 EventGraph {nodes [
   Event BlueprintUpdateAnimation()
   n1 = Get PawnSpeed2D()
   > Set PlayerMoving(PlayerMoving:double=SelectFloat(A=1, B=0, bPickA="float >= float"(A=@n1, B=1) {is=title K2Node_CallFunction})) -> (Output_Get:double)
   TryGetPawnOwner()
  ]}
]
variables [PlayerMoving {type=number}]
